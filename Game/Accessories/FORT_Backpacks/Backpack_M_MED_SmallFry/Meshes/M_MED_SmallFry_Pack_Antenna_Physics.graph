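# Physics asset for the antennas on the SmallFry backpack, with capsule bodies for the pack, the head and both antennas.
# Left and right antennas are each a four-bone chain of thin capsules, constrained link to link and rooted on dny_head.
# All eight joints share very small swing and twist limit angles.
# The large pack capsule has no constraints, and no joint links it to the head.
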PhysicsAsset /Game/Accessories/FORT_Backpacks/Backpack_M_MED_SmallFry/Meshes/M_MED_SmallFry_Pack_Antenna_Physics
bodies {
 dny_antenna_00_l [{shapeType=Capsule, length=1.5659, radius=0.5, center=<-0.0001 -0.7991 -0.1193>, rotation=<rot -88.4674 89.9973 179.9989>}]
 dny_antenna_00_r [{shapeType=Capsule, length=1.5659, radius=0.5, center=<-0.0001 1.5848 0.0424>, rotation=<rot -88.4674 89.9973 179.9989>}]
 dny_antenna_01_l [{shapeType=Capsule, length=2.8958, radius=0.5, center=<-0.0002 -1.1798 -0.1228>, rotation=<rot -84.0587 89.9917 179.9994>}]
 dny_antenna_01_r [{shapeType=Capsule, length=2.8958, radius=0.5, center=<0 1.1948 0.1244>, rotation=<rot -84.0587 89.9917 179.9994>}]
 dny_antenna_02_l [{shapeType=Capsule, length=2.2434, radius=0.5, center=<-0.2236 0.036 0.1367>, rotation=<rot 86.0567 -89.999 -9.3564>}]
 dny_antenna_02_r [{shapeType=Capsule, length=1.9275, radius=0.5, center=<0.0682 0.2435 0.004>, rotation=<rot 89.739 -120.0488 -31.399>}]
 dny_antenna_03_l [{shapeType=Capsule, length=3.1536, radius=0.5, center=<0.0112 0.2512 0.0015>, rotation=<rot 88.9272 89.917 179.9987>}]
 dny_antenna_03_r [{shapeType=Capsule, length=3.102, radius=0.5, center=<0.0539 0.0246 0.0223>, rotation=<rot 85.1611 -80.0917 10.1219>}]
 dny_head [{shapeType=Capsule, length=3.0774, radius=2.3368, center=<0.0024 -0.0867 3.8747>, rotation=<rot 20.0274 -90 0.095>}]
 pack [{shapeType=Capsule, length=23.6378, radius=11.2303, center=<0.0227 10.9657 -7.3085>, rotation=<rot -3.9639 -90 2.2213>}]
}
constraints [
 {bone1Name=dny_antenna_00_l, bone2Name=dny_head, swing1LimitDegrees=0.1, swing2LimitDegrees=0.1, twistLimitDegrees=2}
 {bone1Name=dny_antenna_01_l, bone2Name=dny_antenna_00_l, swing1LimitDegrees=0.1, swing2LimitDegrees=0.1, twistLimitDegrees=2}
 {bone1Name=dny_antenna_02_l, bone2Name=dny_antenna_01_l, swing1LimitDegrees=0.1, swing2LimitDegrees=0.1, twistLimitDegrees=2}
 {bone1Name=dny_antenna_03_l, bone2Name=dny_antenna_02_l, swing1LimitDegrees=0.1, swing2LimitDegrees=0.1, twistLimitDegrees=2}
 {bone1Name=dny_antenna_00_r, bone2Name=dny_head, swing1LimitDegrees=0.1, swing2LimitDegrees=0.1, twistLimitDegrees=2}
 {bone1Name=dny_antenna_01_r, bone2Name=dny_antenna_00_r, swing1LimitDegrees=0.1, swing2LimitDegrees=0.1, twistLimitDegrees=2}
 {bone1Name=dny_antenna_02_r, bone2Name=dny_antenna_01_r, swing1LimitDegrees=0.1, swing2LimitDegrees=0.1, twistLimitDegrees=2}
 {bone1Name=dny_antenna_03_r, bone2Name=dny_antenna_02_r, swing1LimitDegrees=0.1, swing2LimitDegrees=0.1, twistLimitDegrees=2}
]
