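# Physics asset for the body mesh of the SmallFry backpack, a medium male Fortnite back accessory.
# Three capsule bodies cover the large pack volume and two smaller bones, dny_body and dny_body_en.
# Two constraints chain dny_body to dny_body_en and dny_body_en to pack.
# Swing and twist limits are a fraction of a degree, tightest at the pack joint, so the chain is almost rigid in rotation.
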PhysicsAsset /Game/Accessories/FORT_Backpacks/Backpack_M_MED_SmallFry/Meshes/M_MED_SmallFry_Pack_Body_Physics
bodies {
 dny_body [{shapeType=Capsule, length=6.3783, radius=3.5573, center=<0.046 1.4934 -0.3777>, rotation=<rot 0.1634 -90.0318 -0.1724>}]
 dny_body_en [{shapeType=Capsule, length=2.9131, radius=3.7388, center=<-0.0119 -2.9908 0.7528>, rotation=<rot -89.2444 90.001 180>}]
 pack [{shapeType=Capsule, length=23.6378, radius=11.2303, center=<0.0227 10.9657 -7.3085>, rotation=<rot -3.9639 -90 2.2213>}]
}
constraints [
 {bone1Name=dny_body, bone2Name=dny_body_en, swing1LimitDegrees=0.1, swing2LimitDegrees=0.1, twistLimitDegrees=0.5}
 {bone1Name=dny_body_en, bone2Name=pack, swing1LimitDegrees=0.01, swing2LimitDegrees=0.01, twistLimitDegrees=0.05}
]
