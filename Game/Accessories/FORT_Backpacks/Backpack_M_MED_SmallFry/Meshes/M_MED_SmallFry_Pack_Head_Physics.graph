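# Physics asset for the SmallFry medium male backpack (Pack_Head), supplying simple collision bodies for the pack and its dny bone chain.
# Four capsules cover the large pack and the smaller dny_body, dny_neck and dny_head bones.
# Two constraints link head to neck and neck to body with zero swing limits and a very small twist limit.
# The pack body has no constraint to the chain.
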PhysicsAsset /Game/Accessories/FORT_Backpacks/Backpack_M_MED_SmallFry/Meshes/M_MED_SmallFry_Pack_Head_Physics
bodies {
 dny_body [{shapeType=Capsule, length=13.763, radius=3.6788, center=<-0.4389 0.3517 -5.8815>, rotation=<rot -4.8583 -89.7159 -178.8464>}]
 dny_head [{shapeType=Capsule, length=0.5447, radius=0.8846, center=<0 0.0039 -0.035>, rotation=<rot 0 0 4.5961>}]
 dny_neck [{shapeType=Capsule, length=0, radius=1.4783, center=<0.0001 0.1538 -0.1673>, rotation=<rot 4.6311 -89.9998 0>}]
 pack [{shapeType=Capsule, length=23.6378, radius=11.2303, center=<0.0227 10.9657 -7.3085>, rotation=<rot -3.9639 -90 2.2213>}]
}
constraints [
 {bone1Name=dny_neck, bone2Name=dny_body, swing1LimitDegrees=0, swing2LimitDegrees=0, twistLimitDegrees=2}
 {bone1Name=dny_head, bone2Name=dny_neck, swing1LimitDegrees=0, swing2LimitDegrees=0, twistLimitDegrees=2}
]
