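# Physics asset for the SmallFry backpack tail mesh, defining collision bodies and joints for the pack and its two tail bones.
# Each of the three bones has one capsule body, with the largest covering the pack.
# Constraints chain the second tail bone to the first and the first to the pack, with both swing axes nearly locked and only slight twist allowed.
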PhysicsAsset /Game/Accessories/FORT_Backpacks/Backpack_M_MED_SmallFry/Meshes/M_MED_SmallFry_Pack_Tail_Physics
bodies {
 dny_tail_00 [{shapeType=Capsule, length=1.1732, radius=2.2776, center=<-0.0281 -0.237 -2.2347>, rotation=<rot 0 -178.8315 -63.2458>}]
 dny_tail_01 [{shapeType=Capsule, length=7.1766, radius=3.2792, center=<0 1.0854 -7.9563>, rotation=<rot -15.3341 -90 0>}]
 pack [{shapeType=Capsule, length=23.6378, radius=11.2303, center=<0.0227 10.9657 -7.3085>, rotation=<rot -3.9639 -90 2.2213>}]
}
constraints [
 {bone1Name=dny_tail_00, bone2Name=pack, swing1LimitDegrees=0.01, swing2LimitDegrees=0.01, twistLimitDegrees=5}
 {bone1Name=dny_tail_01, bone2Name=dny_tail_00, swing1LimitDegrees=0.01, swing2LimitDegrees=0.01, twistLimitDegrees=5}
]
