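# Physics asset for the SmallFry backpack's wing rig, defining collision bodies and joints for physically simulated secondary motion.
# Capsules cover the pack and a central body bone, while boxes cover two wing pairs (A and B), each wing a three-bone chain per side.
# Constraints link the body to the pack and each wing chain to the body, with tight swing and twist limits and looser twist on middle segments.
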PhysicsAsset /Game/Accessories/FORT_Backpacks/Backpack_M_MED_SmallFry/Meshes/M_MED_SmallFry_Pack_Wing_Physics1
bodies {
 dny_body [{shapeType=Capsule, length=6.3783, radius=3.5573, center=<0.046 1.4934 -0.3777>, rotation=<rot 0.1634 -90.0318 -0.1724>}]
 dny_wingA_00_l [{shapeType=Box, extentX=2.3037, extentY=3.1605, extentZ=3.1605}]
 dny_wingA_00_r [{shapeType=Box, extentX=2.3037, extentY=3.1605, extentZ=3.1605}]
 dny_wingA_01_l [{shapeType=Box, extentX=0.9799, extentY=11.3366, extentZ=6.518, center=<0.0002 -2.6716 -0.3124>}]
 dny_wingA_01_r [{shapeType=Box, extentX=0.9799, extentY=11.0721, extentZ=6.5464, center=<0.0005 2.8946 0>}]
 dny_wingA_02_l [{shapeType=Box, extentX=0.5, extentY=14.7554, extentZ=5.0275, center=<0 1.5247 0>}]
 dny_wingA_02_r [{shapeType=Box, extentX=0.5, extentY=14.7554, extentZ=5.0275, center=<-0.0002 -2.3462 0>}]
 dny_wingB_00_l [{shapeType=Box, extentX=2.3037, extentY=3.1605, extentZ=3.1605}]
 dny_wingB_00_r [{shapeType=Box, extentX=2.3037, extentY=3.1605, extentZ=3.1605}]
 dny_wingB_01_l [{shapeType=Box, extentX=0.9799, extentY=10.3325, extentZ=6.6451, center=<0 -4.2702 0>}]
 dny_wingB_01_r [{shapeType=Box, extentX=0.9799, extentY=10.3325, extentZ=6.6451, center=<-0.0001 3.7049 0>}]
 dny_wingB_02_l [{shapeType=Box, extentX=0.5, extentY=6.6731, extentZ=3.4278}]
 dny_wingB_02_r [{shapeType=Box, extentX=0.5, extentY=6.6731, extentZ=3.4278}]
 pack [{shapeType=Capsule, length=23.6378, radius=11.2303, center=<0.0227 10.9657 -7.3085>, rotation=<rot -3.9639 -90 2.2213>}]
}
constraints [
 {bone1Name=dny_body, bone2Name=pack, swing1LimitDegrees=5, swing2LimitDegrees=5, twistLimitDegrees=15}
 {bone1Name=dny_wingA_00_l, bone2Name=dny_body, swing1LimitDegrees=5, swing2LimitDegrees=2, twistLimitDegrees=2}
 {bone1Name=dny_wingA_01_l, bone2Name=dny_wingA_00_l, swing1LimitDegrees=5, swing2LimitDegrees=5, twistLimitDegrees=15}
 {bone1Name=dny_wingA_02_l, bone2Name=dny_wingA_01_l, swing1LimitDegrees=5, swing2LimitDegrees=2, twistLimitDegrees=2}
 {bone1Name=dny_wingB_00_l, bone2Name=dny_body, swing1LimitDegrees=5, swing2LimitDegrees=2, twistLimitDegrees=2}
 {bone1Name=dny_wingB_01_l, bone2Name=dny_wingB_00_l, swing1LimitDegrees=5, swing2LimitDegrees=5, twistLimitDegrees=10}
 {bone1Name=dny_wingB_02_l, bone2Name=dny_wingB_01_l, swing1LimitDegrees=5, swing2LimitDegrees=2, twistLimitDegrees=2}
 {bone1Name=dny_wingA_00_r, bone2Name=dny_body, swing1LimitDegrees=5, swing2LimitDegrees=2, twistLimitDegrees=2}
 {bone1Name=dny_wingA_01_r, bone2Name=dny_wingA_00_r, swing1LimitDegrees=5, swing2LimitDegrees=5, twistLimitDegrees=15}
 {bone1Name=dny_wingA_02_r, bone2Name=dny_wingA_01_r, swing1LimitDegrees=5, swing2LimitDegrees=2, twistLimitDegrees=2}
 {bone1Name=dny_wingB_00_r, bone2Name=dny_body, swing1LimitDegrees=5, swing2LimitDegrees=2, twistLimitDegrees=2}
 {bone1Name=dny_wingB_01_r, bone2Name=dny_wingB_00_r, swing1LimitDegrees=5, swing2LimitDegrees=5, twistLimitDegrees=10}
 {bone1Name=dny_wingB_02_r, bone2Name=dny_wingB_01_r, swing1LimitDegrees=5, swing2LimitDegrees=2, twistLimitDegrees=2}
]
