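# Animation Blueprint for the SmallFry backpack cosmetic, a winged creature whose output pose is just a shuffled random player of idle and flap-break sequences.
# Event graph watches the kill stat to trigger timed wing flapping, then a surprise phase, alongside sine-driven wing wobble.
# Additive wing bone rotations, RigidBody chains for tail, body, head, antenna and wings, and a flying state machine exist but never reach the output pose.
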AnimBlueprint /BRCosmetics/Accessories/FORT_Backpacks/Backpack_M_MED_SmallFry/Meshes/M_MED_SmallFry_Pack_AnimBP
target_skeleton=/Game/Accessories/FORT_Backpacks/Backpack_M_MED_SmallFry/Meshes/M_MED_SmallFry_Pack_Skeleton
graphs [
 AnimGraph {nodes [
   n1 = Get Rotate_A() -> (Rotate_A:Rotator)
   n2 = "Transform (Modify) Bone - Bone: dny_wingA_00_l"(ComponentPose=Local To Component(LocalPose=Use cached pose 'AnimCache'()), Alpha=1, Translation="0,0,0", Rotation=@n1, Scale="1,1,1") {
    rotationMode=BMM_Additive
    rotationSpace=BCS_BoneSpace
    scaleMode=BMM_Additive
    scaleSpace=BCS_BoneSpace
    alphaBoolBlend {blendOption=Linear}
    boneToModify {boneName=dny_wingA_00_l}
   }
   n3 = Get Rotate_B() -> (Rotate_B:Rotator)
   n4 = "Transform (Modify) Bone - Bone: dny_wingB_00_l"(ComponentPose=@n2, Alpha=1, Translation="0,0,0", Rotation=@n3, Scale="1,1,1") {
    rotationMode=BMM_Additive
    rotationSpace=BCS_BoneSpace
    scaleMode=BMM_Additive
    scaleSpace=BCS_BoneSpace
    alphaBoolBlend {blendOption=Linear}
    boneToModify {boneName=dny_wingB_00_l}
   }
   n5 = "Transform (Modify) Bone - Bone: dny_wingA_00_r"(ComponentPose=@n4, Alpha=1, Translation="0,0,0", Rotation=@n1, Scale="1,1,1") {
    rotationMode=BMM_Additive
    rotationSpace=BCS_BoneSpace
    scaleMode=BMM_Additive
    scaleSpace=BCS_BoneSpace
    alphaBoolBlend {blendOption=Linear}
    boneToModify {boneName=dny_wingA_00_r}
   }
   n6 = "Transform (Modify) Bone - Bone: dny_wingB_00_r"(ComponentPose=@n5, Alpha=1, Translation="0,0,0", Rotation=@n3, Scale="1,1,1") {
    rotationMode=BMM_Additive
    rotationSpace=BCS_BoneSpace
    scaleMode=BMM_Additive
    scaleSpace=BCS_BoneSpace
    alphaBoolBlend {blendOption=Linear}
    boneToModify {boneName=dny_wingB_00_r}
   }
   n7 = Get Flap_WingsA() -> (Flap_WingsA:Rotator)
   n8 = "Transform (Modify) Bone - Bone: dny_wingA_00_l"(ComponentPose=@n6, Alpha=1, Translation="0,0,0", Rotation=@n7, Scale="1,1,1") {
    rotationMode=BMM_Additive
    rotationSpace=BCS_BoneSpace
    scaleMode=BMM_Additive
    scaleSpace=BCS_BoneSpace
    alphaBoolBlend {blendOption=Linear}
    boneToModify {boneName=dny_wingA_00_l}
   }
   n9 = Get Flap_WingsB() -> (Flap_WingsB:Rotator)
   n10 = "Transform (Modify) Bone - Bone: dny_wingB_00_l"(ComponentPose=@n8, Alpha=1, Translation="0,0,0", Rotation=@n9, Scale="1,1,1") {
    rotationMode=BMM_Additive
    rotationSpace=BCS_BoneSpace
    scaleMode=BMM_Additive
    scaleSpace=BCS_BoneSpace
    alphaBoolBlend {blendOption=Linear}
    boneToModify {boneName=dny_wingB_00_l}
   }
   n11 = "Transform (Modify) Bone - Bone: dny_wingA_00_r"(ComponentPose=@n10, Alpha=1, Translation="0,0,0", Rotation=@n7, Scale="1,1,1") {
    rotationMode=BMM_Additive
    rotationSpace=BCS_BoneSpace
    scaleMode=BMM_Additive
    scaleSpace=BCS_BoneSpace
    alphaBoolBlend {blendOption=Linear}
    boneToModify {boneName=dny_wingA_00_r}
   }
   n12 = "Transform (Modify) Bone - Bone: dny_wingB_00_r"(ComponentPose=@n11, Alpha=1, Translation="0,0,0", Rotation=@n9, Scale="1,1,1") {
    rotationMode=BMM_Additive
    rotationSpace=BCS_BoneSpace
    scaleMode=BMM_Additive
    scaleSpace=BCS_BoneSpace
    alphaBoolBlend {blendOption=Linear}
    boneToModify {boneName=dny_wingB_00_r}
   }
   n13 = RigidBody(ComponentPose=@n12) {
    bClampLinearTranslationLimitToRefPose=true
    bOverrideWorldGravity=true
    overridePhysicsAsset=/Game/Accessories/FORT_Backpacks/Backpack_M_MED_SmallFry/Meshes/M_MED_SmallFry_Pack_Tail_Physics
    simulationSpace=BaseBoneSpace
    alphaBoolBlend {blendOption=Linear}
    baseBoneRef {boneName=pack}
    componentAppliedLinearAccClamp=<1500 1500 1500>
    componentLinearAccScale=<0.35 0.45 0.45>
    componentLinearVelScale=<0.3 0.3 0.3>
    overrideWorldGravity {z=-700}
   }
   n14 = RigidBody(ComponentPose=@n13) {
    bClampLinearTranslationLimitToRefPose=true
    bOverrideWorldGravity=true
    overridePhysicsAsset=/Game/Accessories/FORT_Backpacks/Backpack_M_MED_SmallFry/Meshes/M_MED_SmallFry_Pack_Body_Physics
    simulationSpace=BaseBoneSpace
    alphaBoolBlend {blendOption=Linear}
    baseBoneRef {boneName=pack}
    componentAppliedLinearAccClamp=<300 300 300>
    componentLinearAccScale=<0.35 0.45 0.45>
    componentLinearVelScale=<0.3 0.3 0.3>
    overrideWorldGravity {z=-600}
   }
   n15 = RigidBody(ComponentPose=@n14) {
    bClampLinearTranslationLimitToRefPose=true
    bOverrideWorldGravity=true
    overridePhysicsAsset=/Game/Accessories/FORT_Backpacks/Backpack_M_MED_SmallFry/Meshes/M_MED_SmallFry_Pack_Head_Physics
    simulationSpace=BaseBoneSpace
    alphaBoolBlend {blendOption=Linear}
    baseBoneRef {boneName=pack}
    componentAppliedLinearAccClamp=<300 300 300>
    componentLinearAccScale=<0.35 0.45 0.45>
    componentLinearVelScale=<0.3 0.3 0.3>
   }
   n16 = RigidBody(ComponentPose=@n15) {
    bClampLinearTranslationLimitToRefPose=true
    bOverrideWorldGravity=true
    overridePhysicsAsset=/Game/Accessories/FORT_Backpacks/Backpack_M_MED_SmallFry/Meshes/M_MED_SmallFry_Pack_Antenna_Physics
    simulationSpace=BaseBoneSpace
    alphaBoolBlend {blendOption=Linear}
    baseBoneRef {boneName=pack}
    componentAppliedLinearAccClamp=<3000 3000 3000>
    componentLinearAccScale=<0.35 0.45 0.45>
    componentLinearVelScale=<0.3 0.3 0.3>
    overrideWorldGravity {z=-150}
   }
   n17 = RigidBody(ComponentPose=@n16) {
    bClampLinearTranslationLimitToRefPose=true
    bOverrideWorldGravity=true
    overridePhysicsAsset=/Game/Accessories/FORT_Backpacks/Backpack_M_MED_SmallFry/Meshes/M_MED_SmallFry_Pack_Wing_Physics1
    simulationSpace=BaseBoneSpace
    alphaBoolBlend {blendOption=Linear}
    baseBoneRef {boneName=pack}
    componentAppliedLinearAccClamp=<1200 1200 1200>
    componentLinearAccScale=<0.35 0.45 0.45>
    componentLinearVelScale=<0.3 0.3 0.3>
    overrideWorldGravity {z=-150}
   }
   Component To Local(ComponentPose=@n17)
   Local To Component(LocalPose=Use cached pose 'AnimCache'())
   Local To Component(LocalPose=Two Way Blend!(A=Component To Local(), B=Component To Local(), bAlphaBoolEnabled:bool=Get Surprise_Out() -> (Surprise_Out:bool)) -> (Pose) {blendNode {alphaInputType=Bool, alphaBoolBlend {blendInTime=0.3, blendOption=CircularInOut, blendOutTime=0.3}}})
   n18 = Random Sequence Player!() -> (Pose:Pose Link Structure) {
    is=title AnimGraphNode_RandomPlayer
    bShuffleMode=true
    entries [
     {
      chanceToPlay=0.75
      maxLoopCount=3
      maxPlayRate=1
      minLoopCount=2
      minPlayRate=1
      sequence=/BRCosmetics/Accessories/FORT_Backpacks/Backpack_M_MED_SmallFry/Animation/Idle_SmallFry_Pack
      blendIn {blendOption=HermiteCubic, blendTime=0.2, customCurve=null}
     }
     {
      chanceToPlay=0.25
      maxLoopCount=1
      maxPlayRate=1
      minLoopCount=0
      minPlayRate=1
      sequence=/BRCosmetics/Accessories/FORT_Backpacks/Backpack_M_MED_SmallFry/Animation/IdleBreak_Flap_SmallFry_Pack
      blendIn {blendOption=HermiteCubic, blendTime=0.2, customCurve=null}
     }
    ]
   }
   Output Pose(Result=@n18) {layerGroup=None}
   Save cached pose 'AnimCache'(Pose=Local Space Ref Pose()) {cacheName=AnimCache}
   AnimGraphNode_StateMachine_0 = AnimStates() {is=title AnimGraphNode_StateMachine}
  ]}
 AnimStates {
  entry=Idle
  path=AnimGraph.AnimGraphNode_StateMachine_0.AnimStates
  ids=id
  nodes [
   AnimStateNode_1 = IdleBreak() {is=title AnimStateNode}
   AnimStateNode_2 = IdleBreak_Start() {is=title AnimStateNode}
   AnimStateNode_3 = IdleBreak_Outro() {is=title AnimStateNode}
   AnimStateNode_4 = Idle()
   AnimStateTransitionNode_2 = AnimStateTransitionNode()
   AnimStateTransitionNode_3 = AnimStateTransitionNode() {bAutomaticRuleBasedOnSequencePlayerInState=true, crossfadeDuration=1}
   AnimStateTransitionNode_4 = AnimStateTransitionNode()
   AnimStateTransitionNode_5 = AnimStateTransitionNode() {bAutomaticRuleBasedOnSequencePlayerInState=true}
  ]
  links [
   {
    dst=IdleBreak_Start
    rule=AnimGraph.AnimGraphNode_StateMachine_0.AnimStates.AnimStateTransitionNode_2.Transition
    src=Idle
    via=AnimStateTransitionNode_2
   }
   {
    dst=IdleBreak_Outro
    rule=AnimGraph.AnimGraphNode_StateMachine_0.AnimStates.AnimStateTransitionNode_4.Transition
    src=IdleBreak
    via=AnimStateTransitionNode_4
   }
   {
    dst=Idle
    rule=AnimGraph.AnimGraphNode_StateMachine_0.AnimStates.AnimStateTransitionNode_5.Transition
    src=IdleBreak_Outro
    via=AnimStateTransitionNode_5
   }
   {
    dst=IdleBreak
    rule=AnimGraph.AnimGraphNode_StateMachine_0.AnimStates.AnimStateTransitionNode_3.Transition
    src=IdleBreak_Start
    via=AnimStateTransitionNode_3
   }
  ]
 }
 IdleBreak {
  path=AnimGraph.AnimGraphNode_StateMachine_0.AnimStates.AnimStateNode_1.IdleBreak
  nodes [Output Animation Pose(Result=Sequence Player 'Flying_Loop_SmallFry_Pack'() {sequence=/Game/Accessories/FORT_Backpacks/Backpack_M_MED_SmallFry/Animation/NotUsed/Flying_Loop_SmallFry_Pack})]
 }
 IdleBreak_Start {
  path=AnimGraph.AnimGraphNode_StateMachine_0.AnimStates.AnimStateNode_2.IdleBreak_Start
  nodes [
   n1 = Sequence Player 'Flying_Intro_SmallFry_Pack'() {bLoopAnimation=false, sequence=/Game/Accessories/FORT_Backpacks/Backpack_M_MED_SmallFry/Animation/NotUsed/Flying_Intro_SmallFry_Pack}
   Output Animation Pose(Result=@n1)
  ]
 }
 IdleBreak_Outro {
  path=AnimGraph.AnimGraphNode_StateMachine_0.AnimStates.AnimStateNode_3.IdleBreak_Outro
  nodes [Output Animation Pose(Result=Sequence Player 'Flying_Outro_SmallFry_Pack'() {sequence=/Game/Accessories/FORT_Backpacks/Backpack_M_MED_SmallFry/Animation/NotUsed/Flying_Outro_SmallFry_Pack})]
 }
 Idle {
  path=AnimGraph.AnimGraphNode_StateMachine_0.AnimStates.AnimStateNode_4.Idle
  nodes [
   Random Sequence Player() {entries [
     {
      chanceToPlay=0.5
      maxLoopCount=0
      maxPlayRate=1
      minLoopCount=0
      minPlayRate=1
      sequence=&Idle_SmallFry_Pack
      blendIn {blendOption=HermiteCubic, blendTime=0.2, customCurve=null}
     }
     {
      chanceToPlay=0.5
      maxLoopCount=0
      maxPlayRate=1
      minLoopCount=0
      minPlayRate=1
      sequence=&IdleBreak_Flap_SmallFry_Pack
      blendIn {blendOption=HermiteCubic, blendTime=0.2, customCurve=null}
     }
    ]}
   Output Animation Pose(Result=Sequence Player 'Idle_SmallFry_Pack'() {sequence=&Idle_SmallFry_Pack})
  ]
 }
 Transition {
  path=AnimGraph.AnimGraphNode_StateMachine_0.AnimStates.AnimStateTransitionNode_2.Transition
  nodes [
   n1 = Get bIsOnGround() -> (bIsOnGround:bool)
   Result(bCanEnterTransition=OR Boolean(A=NOT Boolean(A=@n1), B=Get bIsPlayingEmote()))
   Get bIsSurfaceSwimming()
   Get bIsSkyDiving()
   Get bIsSkydiveDiveMode()
   Get bIsSkydiveFloating()
  ]
 }
 Transition {path=AnimGraph.AnimGraphNode_StateMachine_0.AnimStates.AnimStateTransitionNode_3.Transition, nodes [Result()]}
 Transition {
  path=AnimGraph.AnimGraphNode_StateMachine_0.AnimStates.AnimStateTransitionNode_4.Transition
  nodes [
   n1 = Get bIsSurfaceSwimming()
   n2 = Get bIsPlayingEmote()
   Result(bCanEnterTransition=AND Boolean!(A=NOT Boolean(A=@n1), B=Get bIsOnGround(), C:bool=NOT Boolean(A=@n2)) -> (ReturnValue))
   NOT Boolean(A=Get bIsSkydiveFloating())
   NOT Boolean(A=Get bIsSkyDiving())
   NOT Boolean(A=Get bIsSkydiveDiveMode())
   OR Boolean(A=false, B=false)
  ]
 }
 Transition {path=AnimGraph.AnimGraphNode_StateMachine_0.AnimStates.AnimStateTransitionNode_5.Transition, nodes [Result()]}
 EventGraph {nodes [
   Event OnCharacterCustomizationLoaded()
   > n1 = Cast To FortPlayerPawn(Object=TryGetFortPawn())
   > n2 = WatchCosmeticStat!(
    execute:Exec
    Pawn:Fort Player Pawn Object Reference=@n1
    StatTag:Fort Stat Manager Tag Structure=(TagName="StatManager.AthenaKills")
    DefaultValue:int=0
    bBroadcastInitialValue:bool=true
    InitialValueBroadcastDelay:float=-1
    ChangeValueBroadcastDelay:float=-1
    bConsumeWaitingNotifications:bool=true
   ) -> (then:Exec, OnStatChanged:Exec, StatName:Name, StatValue:int) {is=title K2Node_AsyncAction}
   Set KillCount(execute=@n2.OnStatChanged, KillCount:double="To Float (Integer)"(InInt=@n2.StatValue)) -> (Output_Get:double)
   > Set Active_Kill(Active_Kill:bool=true) -> (Output_Get:bool)
   n3 = Event BlueprintUpdateAnimation()
   > Get!(execute:Exec) -> (then:Exec, else:Exec, Skel_Mesh:Skeletal Mesh Component Object Reference)
   > n4 = Sequence!(execute) -> (then_0, then_1, then_2:Exec)
   > n5 = Branch(Condition=Get bIsSkyDiving())
   > Set Flap_WingsA(Flap_WingsA:Rotator="0, 0, 0") -> (Output_Get:Rotator)
   > Set Flap_WingsB(Flap_WingsB:Rotator="0, 0, 0") -> (Output_Get:Rotator)
   n6 = Branch(execute=@n5.else, Condition=Get Surprise_Out())
   > n7 = Set Surprise_Timer(Surprise_Timer:double=float + float(A=Get Surprise_Timer() -> (Surprise_Timer:double), B=1) {is=title K2Node_CommutativeAssociativeBinaryOperator}) -> (Output_Get:double)
   > Branch(Condition="Equal (Float)"!(A:double=@n7, B:double=Get Surprise_Length_Out() -> (Surprise_Length_Out:double)) -> (ReturnValue:bool) {is=title K2Node_CallFunction})
   > Set Surprise_Out(Surprise_Out:bool=false) -> (Output_Get:bool)
   > Set Surprise_Timer(Surprise_Timer=0)
   > Set Crank_Timer(Crank_Timer:double=0) -> (Output_Get:double)
   > Set Crank_Timer_Reduce(Crank_Timer_Reduce:double=0) -> (Output_Get:double)
   Set Flap_WingsA(
    execute=[@n6.else, @n16.then]
    Flap_WingsA=RInterpTo_Constant!(Current:Rotator=Get Flap_WingsA(), Target:Rotator="0,-40,-60", DeltaTime:float=@n3.DeltaTimeX, InterpSpeed:float=100) -> (ReturnValue:Rotator)
   )
   > Set Flap_WingsB(Flap_WingsB=RInterpTo_Constant(Current=Get Flap_WingsB(), Target="0,-30,-10", DeltaTime=@n3.DeltaTimeX, InterpSpeed=70))
   > Set Surprise_Length_Out(Surprise_Length_Out:double=205) -> (Output_Get:double)
   n8 = GetGameTimeInSeconds()
   n9 = "Sin (Radians)"(A=float * float(A=float * float(A=0.15, B=10), B=@n8))
   n10 = "Sin (Radians)"(A=float * float(A=float * float(A=0.17, B=8), B=@n8))
   n11 = "Sin (Radians)"(A=float * float(A=float * float(A=0.13, B=13), B=@n8))
   Set Rotate_A(
    execute=@n4.then_2
    Rotate_A:Rotator=MakeRotator(Roll=float * float(A=@n9, B=4.2), Pitch=float * float(A=@n10, B=5.2), Yaw=float * float(A=@n11, B=3.2))
   ) -> (Output_Get:Rotator)
   n12 = GetGameTimeInSeconds()
   n13 = "Sin (Radians)"(A=float * float(A=float * float(A=0.1, B=8), B=@n12))
   n14 = "Sin (Radians)"(A=float * float(A=float * float(A=0.13, B=11), B=@n12))
   n15 = "Sin (Radians)"(A=float * float(A=float * float(A=0.07, B=10), B=@n12))
   > n16 = Set Rotate_B(Rotate_B:Rotator=MakeRotator(Roll=float * float(A=@n15, B=4), Pitch=float * float(A=@n13, B=5), Yaw=float * float(A=@n14, B=3))) -> (Output_Get:Rotator)
   Branch(execute=@n4.then_1, Condition=Get Active_Kill() -> (Active_Kill:bool))
   n17 = float + float(A=40, B=Get Crank_Timer_Reduce() -> (Crank_Timer_Reduce:double))
   n18 = float * float(A=@n17, B=GetGameTimeInSeconds())
   n19 = MakeRotator(Yaw=MapRangeClamped(Value="Sin (Radians)"(A=@n18), InRangeA=-1, InRangeB=1, OutRangeA=-40, OutRangeB=@n17))
   > Set Flap_WingsB(Flap_WingsB=@n19)
   > Set Flap_WingsA(Flap_WingsA=@n19)
   > n20 = Set Crank_Timer(Crank_Timer=float + float(A=Get Crank_Timer() -> (Crank_Timer:double), B=1))
   > Branch(Condition="float >= float"(A=@n20, B=130) {is=title K2Node_CallFunction})
   > Set Crank_Timer_Reduce(Crank_Timer_Reduce=-20)
   > Branch(Condition="float >= float"(A=@n20, B=200))
   > Set Active_Kill(Active_Kill=false)
   > Set Surprise_Out(Surprise_Out=true)
   Event BlueprintInitializeAnimation()
   > Set Skel_Mesh(Skel_Mesh=GetOwnerSkeletalMeshComponent())
   TryGetPawnOwner()
  ]}
]
variables [
 Active_Kill {type=boolean}
 Crank_time_rot {type=number}
 Crank_Timer {type=number}
 Crank_Timer_Reduce {type=number}
 Flap_WingsA {type=Rotator}
 Flap_WingsB {type=Rotator}
 KillCount {type=number}
 Rotate_A {type=Rotator}
 Rotate_B {type=Rotator}
 Skel_Mesh {type=SkeletalMeshComponent}
 Surprise_Length_Out {type=number}
 Surprise_Out {type=boolean}
 Surprise_Timer {type=number}
]
